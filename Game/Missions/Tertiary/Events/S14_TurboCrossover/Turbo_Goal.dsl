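; Turbo_Goal is a goal actor for the S14 TurboCrossover event that celebrates when something enters its goal collision.
; An overlap on Goal-ProjectileCollision prints a "TURBO GOAL" message, moves the soccer scored-goal effect component to the overlapping actor's location and activates it with a reset.
; The overlap handler does no actor filtering or score tracking, and the construction script is empty.
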(event OnComponentBeginOverlap(Goal-ProjectileCollision)
  (params (OverlappedComponent :type Primitive%20Component%20Object%20Reference) (OtherActor :type Actor%20Object%20Reference) (OtherComp :type Primitive%20Component%20Object%20Reference) (OtherBodyIndex :type Integer) (bFromSweep :type Boolean) (SweepResult :type Hit%20Result%20Structure%20%28by%20ref%29))
  (bind _p_athena_soccergame_scoredgoal (Variables|Default|GetP_Athena_SoccerGame_ScoredGoal))
  (Development|PrintText "INVTEXT(\"TURBO GOOOOOAAAAALLLL!!!!\")" false)
  (Transformation|SetWorldLocation _p_athena_soccergame_scoredgoal (Transformation|GetActorLocation OtherActor) false false)
  (Components|Activation|Activate _p_athena_soccergame_scoredgoal true) ; VFX - GOAL!
)

(fn ConstructionScript)
